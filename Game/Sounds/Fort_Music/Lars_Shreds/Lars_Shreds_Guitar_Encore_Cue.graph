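# Sound cue that plays the Lars Shreds guitar encore music clip as an in-world game sound.
# A single wave player feeds the Lars_Shreds_Guitar_Encore wave straight to the output, with no mixing, randomization or modulation.
# It uses the close-prop attenuation preset for spatial falloff and routes through the mission in-game sounds class.
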SoundCue /Game/Sounds/Fort_Music/Lars_Shreds/Lars_Shreds_Guitar_Encore_Cue
root=WavePlayer_0
nodes [{id=WavePlayer_0, properties {soundWaveAssetPtr=&/Lars_Shreds_Guitar_Encore}}]
properties {
 attenuationSettings=/Game/Sounds/Attenuation/Fort_Prop_Close_Att
 maxDistance=1800
 soundClassObject=/Game/Sounds/FortSoundClassesAndModes/Fort_Mission_IngameSounds
}
